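# Physics asset for the F_MED_Python back bling, giving its backpack mesh collision bodies and jointed dangling parts.
# A box body covers the pack, with small box bodies at the grenade and zipper base bones, a capsule for the grenade, and boxes for the pin and zipper.
# Tightly limited constraints chain the pin to the grenade and the grenade to its base, and join the zipper to its base.
# The pin joint locks one swing axis.
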PhysicsAsset /BRCosmetics/Accessories/FORT_Backpacks/Backpack_F_MED_Python/Meshes/F_MED_Python_Pack_PhysicsAsset
bodies {
 dyn_grenade_base [{shapeType=Box, extentX=0.5, extentY=0.5, extentZ=0.5}]
 dyn_grenade_parent [{shapeType=Capsule, length=5.4344, radius=2.639, center=<3.0053 -0.0001 -1.2417>, rotation=<rot 90 0 0>}]
 dyn_grenade_pin [{shapeType=Box, extentX=4.0107, extentY=3.5518, extentZ=1.0351, center=<1.5329 0 0>}]
 dyn_zipper_base [{shapeType=Box, extentX=0.5, extentY=0.5, extentZ=0.5}]
 dyn_zipper_parent [{shapeType=Box, extentX=3.5325, extentY=2.5, extentZ=1, center=<1.7834 0 0>}]
 pack [{shapeType=Box, extentX=18.3241, extentY=12.9208, extentZ=21.7614, center=<-0.0757 -4.7904 -1.2385>, rotation=<rot 7.2889 -1.1567 -9.0429>}]
}
constraints [
 {bone1Name=dyn_zipper_parent, bone2Name=dyn_zipper_base, swing1LimitDegrees=10, swing2LimitDegrees=40, twistLimitDegrees=3}
 {
  bone1Name=dyn_grenade_pin
  bone2Name=dyn_grenade_parent
  swing1LimitDegrees=10
  swing1Motion=Locked
  swing2LimitDegrees=30
  twistLimitDegrees=3
 }
 {bone1Name=dyn_grenade_parent, bone2Name=dyn_grenade_base, swing1LimitDegrees=10, swing2LimitDegrees=10, twistLimitDegrees=3}
]
